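; Looping gameplay cue that plays the visuals and audio for the secondary spray of the HealSpray item.
; On start it attaches a Niagara system and a first- or third-person sound loop to the weapon's nozzle and mesh.
; Tick feeds player velocity to the effect, and a local-ammo event binding drives the loop's sound parameter with remaining ammo.
; Removal zeroes AddSource, deactivates the effect, fades the loop and plays a removal sound.
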(event Custom|Local_Ammo
  (params (LocalCount :type Integer) (LocalRemaining :type Integer))
  (bind _audiocomponent_sprayloop_heal (Variables|Default|GetAudioComponent_SprayLoop_Heal)
    (:then
      (Audio|Parameter|SetFloatParameter _audiocomponent_sprayloop_heal "Spray_Loop" (Math|Conversions|ToFloat(Integer) LocalRemaining) :!class /Script/AudioExtensions.AudioParameterControllerInterface))
    (:else)))

(event EventTick
  (params (DeltaSeconds :type Float%20%28single-precision%29))
  (bind _sprayfx (Variables|Default|GetSprayFX)
    (:then
      (Effects|Components|ParticleSystem|SetVectorParameter _sprayfx "PlayerVelocity" (Transformation|GetVelocity (Variables|Default|GetPlayer%20Pawn))))
    (:else)))

(event EventOnLoopingStart(EitherVFX-preferthisversion)
  (params (MyTarget :type Actor%20Object%20Reference) (Parameters :type Gameplay%20Cue%20Parameters%20Structure%20%28by%20ref%29) (ParticleComponents :type Array%20of%20FXSystem%20Component%20Object%20References) (AudioComponents :type Array%20of%20Audio%20Component%20Object%20References))
  (bind _asfort_player_pawn (Utilities|Casting|CastToFortPlayerPawn MyTarget))
  (bind _output_get (Variables|Default|SetPlayer%20Pawn _asfort_player_pawn))
  (CallFunction|StartSprayAudio)
  (bind _returnvalue (Niagara|SpawnSystemAttached "/FlipperGameplay/Items/HealSpray/FX/NS_Consumable_HealSpray_Secondary.NS_Consumable_HealSpray_Secondary" (Variables|/Script/FortniteGame.FortWeapon|GetWeaponMesh (Weapon|GetCurrentWeapon _output_get)) "FX_Nozzle" "0, 0, 0" "0, 0, 0" "KeepRelativeOffset" true))
  (Variables|Default|SetSprayFX _returnvalue))

(event EventOnRemoval(EitherVFX-preferthisversion)
  (params (MyTarget :type Actor%20Object%20Reference) (Parameters :type Gameplay%20Cue%20Parameters%20Structure%20%28by%20ref%29) (ParticleComponents :type Array%20of%20FXSystem%20Component%20Object%20References) (AudioComponents :type Array%20of%20Audio%20Component%20Object%20References) (BurstCameraShakeInstance :type Camera%20Shake%20Base%20Object%20Reference) (BurstDecalInstance :type Decal%20Actor%20Object%20Reference))
  (CallFunction|StopSprayAudio)
  (bind _sprayfx (Variables|Default|GetSprayFX)
    (:then
      (Effects|Components|ParticleSystem|SetFloatParameter _sprayfx "AddSource" 0.0)
      (Components|Activation|Deactivate _sprayfx))
    (:else)))

(fn StartSprayAudio
  (bind _player_pawn (Variables|Default|GetPlayer%20Pawn))
  (bind _returnvalue (Weapon|GetCurrentWeapon _player_pawn))
  (bind _audiocomponent_sprayloop_heal (Variables|Default|GetAudioComponent_SprayLoop_Heal)
    (:then
      (Audio|Components|Audio|FadeOut _audiocomponent_sprayloop_heal (Variables|Sound|GetFade%20Out%20Duration) 0.0 "Linear" :!class /Script/Engine.AudioComponent))
    (:else
      _))
  (bind _returnvalue_2 (Audio|SpawnSoundAttached (select (Pawn|IsLocallyViewed _player_pawn) (Variables|Default|GetSpraySound1p) (Variables|Default|GetSpraySound3p)) (Variables|/Script/FortniteGame.FortWeapon|GetWeaponMesh _returnvalue)))
  (Variables|Default|SetAudioComponent_SprayLoop_Heal _returnvalue_2)
  (EventDispatchers|BindEventtoOnLocalAmmoChanged _returnvalue (EventDispatchers|CreateEvent :!event Local_Ammo)))

(fn StopSprayAudio
  (bind _player_pawn (Variables|Default|GetPlayer%20Pawn)
    (:then
      (bind _sprayremoval (Variables|Default|GetSprayRemoval)
        (:then
          (Audio|SpawnSoundAttached _sprayremoval (Variables|/Script/FortniteGame.FortWeapon|GetWeaponMesh (Weapon|GetCurrentWeapon _player_pawn))))
        (:else
          _))
      (bind _audiocomponent_sprayloop_heal (Variables|Default|GetAudioComponent_SprayLoop_Heal)
        (:then
          (Audio|Components|Audio|FadeOut _audiocomponent_sprayloop_heal (Variables|Sound|GetFade%20Out%20Duration) 0.0 "Linear" :!class /Script/Engine.AudioComponent))
        (:else)))
    (:else
      _)))

(fn ConstructionScript)
