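; Looping GameplayCue notify for the Athena shield aura heal that visualizes the link between the healed target and its source.
; On loop start it caches the target and the effect causer from the cue's effect context.
; Each tick, if both are valid and distinct and the target's shield is below the ShieldCap scalable float, it draws a short-lived cyan debug line between them.
; The cue itself applies no healing.
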(event EventBeginPlay)

(event Collision|EventActorBeginOverlap
  (params (OtherActor :type Actor%20Object%20Reference)))

(event EventTick
  (params (DeltaSeconds :type Float%20%28single-precision%29))
  ;[1 1 1] On Tick
  (Utilities|IsValid (Variables|Default|GetTargetActor)
    (:Is%20Valid
      (Utilities|IsValid (Variables|Default|GetSourceActor)
        (:Is%20Valid
          (if (and (Utilities|NotEqual(Object) (Variables|Default|GetTargetActor) (Variables|Default|GetSourceActor)) (< (FortPawn|GetShield (Utilities|Casting|CastToFortPawn (Variables|Default|GetTargetActor))) (ScalableFloat|GetValueatLevel (Variables|Default|GetRow_ShieldCap))))
            (Rendering|Debug|DrawDebugLine (Variables|Default|GetTargetActor) (Transformation|GetActorLocation (Variables|Default|GetTargetActor)) (Transformation|GetActorLocation (Variables|Default|GetSourceActor)) (Math|Color|MakeColor 0.0 255.0 255.0) 0.1 8.0)))))))

(event EventOnLoopingStart(EitherVFX-preferthisversion)
  (params (MyTarget :type Actor%20Object%20Reference) (Parameters :type Gameplay%20Cue%20Parameters%20Structure%20%28by%20ref%29) (ParticleComponents :type Array%20of%20FXSystem%20Component%20Object%20References) (AudioComponents :type Array%20of%20Audio%20Component%20Object%20References))
  (bind (_normalizedmagnitude _rawmagnitude _effectcontext) (Ability|GameplayCue|BreakGameplayCueParameters Parameters))
  ;[1 1 1] On Looping Start
  (Variables|Default|SetTargetActor MyTarget)
  (Variables|Default|SetSourceActor (Ability|EffectContext|GetEffectCauser _effectcontext)))

(fn ConstructionScript)
