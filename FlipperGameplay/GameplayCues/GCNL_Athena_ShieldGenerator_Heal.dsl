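; Looping gameplay cue for the Shield Generator heal effect, playing a shield loop sound on the affected player pawn.
; On looping start it attaches either the standard or third-person (3P) loop to the pawn's root, depending on whether the pawn is locally viewed, and stores the audio component.
; On removal it briefly fades out the stored sound if still valid.
; Both events also call their parent implementations.
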(event EventOnLoopingStart(EitherVFX-preferthisversion)
  (params (MyTarget :type Actor%20Object%20Reference) (Parameters :type Gameplay%20Cue%20Parameters%20Structure%20%28by%20ref%29) (ParticleComponents :type Array%20of%20FXSystem%20Component%20Object%20References) (AudioComponents :type Array%20of%20Audio%20Component%20Object%20References))
  (|CallParentFunction MyTarget Parameters ParticleComponents AudioComponents)
  (bind _asfort_player_pawn (Utilities|Casting|CastToFortPlayerPawn MyTarget))
  (bind _returnvalue (Audio|SpawnSoundAttached (select (Pawn|IsLocallyViewed _asfort_player_pawn) "/FlipperGameplay/Items/ShieldGenerator/Sounds/ShieldGenerator_Shield_Loop_Cue.ShieldGenerator_Shield_Loop_Cue" "/FlipperGameplay/Items/ShieldGenerator/Sounds/ShieldGenerator_Shield_Loop_3P_Cue.ShieldGenerator_Shield_Loop_3P_Cue") (Variables|/Script/Engine.Actor|GetRootComponent _asfort_player_pawn)))
  (Variables|Default|SetHealingSound _returnvalue))

(event EventOnRemoval(EitherVFX-preferthisversion)
  (params (MyTarget :type Actor%20Object%20Reference) (Parameters :type Gameplay%20Cue%20Parameters%20Structure%20%28by%20ref%29) (ParticleComponents :type Array%20of%20FXSystem%20Component%20Object%20References) (AudioComponents :type Array%20of%20Audio%20Component%20Object%20References) (BurstCameraShakeInstance :type Camera%20Shake%20Base%20Object%20Reference) (BurstDecalInstance :type Decal%20Actor%20Object%20Reference))
  (bind _healingsound (Variables|Default|GetHealingSound))
  (|CallParentFunction MyTarget Parameters ParticleComponents AudioComponents BurstCameraShakeInstance BurstDecalInstance)
  (Utilities|IsValid _healingsound
    (:Is%20Valid
      (Audio|Components|Audio|FadeOut _healingsound 0.25 0.0 "Linear" :!class /Script/Engine.AudioComponent))))

(fn ConstructionScript)
